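# Animation Blueprint for the WingBath Sturdy character's face accessory part, posing it from the body and a Control Rig.
# Layers a body-mesh pose copy over the attached-parent copy at clavicle and upper-arm branches, then feeds the rig.
# Event Graph eases opposing on/off alphas for skydive dive mode and emotes using FInterpTo.
# These alphas drive the rig alongside skydiving and DBNO states and a Pauldron flag set from a cosmetic meta tag.
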AnimBlueprint /BRCosmetics/Characters/Player/Male/Medium/Bodies/M_MED_WingBath_Sturdy/Meshes/Parts/M_MED_WingBath_Sturdy_FaceAcc_AnimBP
target_skeleton=&/M_MED_WingBath_Sturdy_FaceAcc_Skeleton
graphs [
 AnimGraph {nodes [
   n1 = Layered blend per bone(
    BasePose=Copy Pose From Mesh() {bUseAttachedParent=true}
    BlendPoses_0=Copy Pose From Mesh(SourceMeshComponent=Get Body Mesh() -> (Body Mesh:Skeletal Mesh Component Object Reference))
   ) {layerSetup [{branchFilters [{boneName=clavicle_l}, {blendDepth=3, boneName=clavicle_r}, {blendDepth=3, boneName=upperarm_r}, {blendDepth=3, boneName=upperarm_l}]}]}
   n2 = Control Rig!(
    Source=@n1
    Alpha
    Is dive mod on alpha:double=Get Divemod On() -> (Divemod On:double)
    Is dive mod off alpha:double=Get Divemod Off() -> (Divemod Off:double)
    Is emote on:double=Get Emote On() -> (Emote On:double)
    Is emote off:double=Get Emote Off() -> (Emote Off:double)
    SkyAlpha:bool=Get bIsSkyDiving()
    DBNO:bool=Get bIsDBNO()
    Pauldron:bool=Get Pauldron() -> (Pauldron:bool)
   ) -> (Pose) {
    alphaBoolBlend {blendOption=Linear}
    controlRigAssetReference {blueprintRigClass=&/M_MED_WingBath_Sturdy_FaceAcc_CtrlRig.M_MED_WingBath_Sturdy_FaceAcc_CtrlRig_C}
   }
   Output Pose(Result=@n2) {layerGroup=None}
  ]}
 EventGraph {nodes [
   n1 = Event BlueprintUpdateAnimation()
   > Set Delta Time X(Delta Time X=@n1.DeltaTimeX)
   > n2 = Sequence!(execute) -> (then_0, then_1, then_2:Exec)
   > n3 = Branch(Condition=Get bIsSkydiveDiveMode())
   n4 = float / float(A=Get Delta Time X(), B=1.5)
   > Set Divemod on(Divemod on:double=FInterpTo(Current=Get Divemod on() -> (Divemod on:double), Target=1, DeltaTime=@n4, InterpSpeed=12)) -> (Output_Get:double)
   Set Divemod on(execute=@n3.else, Divemod on=FInterpTo(Current=Get Divemod on(), Target=0, DeltaTime=@n4, InterpSpeed=12))
   n5 = Sequence!(execute=@n2.then_2) -> (then_0, then_1)
   > n6 = Branch(Condition=Get bIsPlayingEmote())
   n7 = float / float(A=Get Delta Time X(), B=3)
   > Set Emote on(Emote on:double=FInterpTo(Current=Get Emote on() -> (Emote on:double), Target=1, DeltaTime=@n7, InterpSpeed=6)) -> (Output_Get:double)
   Set Emote on(execute=@n6.else, Emote on=FInterpTo(Current=Get Emote on(), Target=0, DeltaTime=@n7, InterpSpeed=6))
   n8 = Branch(execute=@n5.then_1, Condition=NOT Boolean(A=Get bIsPlayingEmote()))
   n9 = float / float(A=Get Delta Time X(), B=3)
   > Set Emote off(Emote off:double=FInterpTo(Current=Get Emote off() -> (Emote off:double), Target=1, DeltaTime=@n9, InterpSpeed=6)) -> (Output_Get:double)
   Set Emote off(execute=@n8.else, Emote off=FInterpTo(Current=Get Emote off(), Target=0, DeltaTime=@n9, InterpSpeed=6))
   n10 = Branch(execute=@n2.then_1, Condition=NOT Boolean(A=Get bIsSkydiveDiveMode()))
   n11 = float / float(A=Get Delta Time X(), B=1.5)
   > Set Divemod off(Divemod off:double=FInterpTo(Current=Get Divemod off() -> (Divemod off:double), Target=1, DeltaTime=@n11, InterpSpeed=12)) -> (Output_Get:double)
   Set Divemod off(execute=@n10.else, Divemod off=FInterpTo(Current=Get Divemod off(), Target=0, DeltaTime=@n11, InterpSpeed=12))
   n12 = Event OnCharacterCustomizationLoaded()
   Set Body Mesh(execute=[@n12.then, @n16.then], Body Mesh:Skeletal Mesh Component Object Reference=GetOwnerSkeletalMeshComponent()) -> (Output_Get:Skeletal Mesh Component Object Reference)
   n13 = TryGetPawnOwner()
   > Is Valid(InputObject=@n13)
   > n14 = Cast To FortPlayerPawnAthena(Object=@n13)
   > Set As Fort Player Pawn Athena(As Fort Player Pawn Athena:Fort Player Pawn Athena Object Reference=@n14) -> (Output_Get:Fort Player Pawn Athena Object Reference)
   > n15 = Cast To PlayerPawn_Athena(Object=@n13)
   > Set Player Pawn(Player Pawn:Player Pawn Athena Object Reference=@n15) -> (Output_Get:Player Pawn Athena Object Reference)
   > Set Pauldron(Pauldron:bool=CharacterItemDefinitionHasAnyMetaTag(Tags=Make Array("[0]"=(TagName="cosmetics.MetaTags.WingBath_Pauldron")))) -> (Output_Get:bool)
   n16 = Event BlueprintInitializeAnimation()
  ]}
]
variables [
 As Fort Player Pawn Athena {type=FortPlayerPawnAthena}
 Body Mesh {type=SkeletalMeshComponent}
 Delta Time X {type=number}
 Divemod Off {type=number}
 Divemod On {type=number}
 Emote Off {type=number}
 Emote On {type=number}
 Pauldron {type=boolean}
 Player Pawn {type=PlayerPawn_Athena_C}
 Sky Dive Alpha {type=number, default=1}
]
