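# Animation Blueprint for the head part of the M_MED_WingBath_Sturdy cosmetic body, on its FaceAcc skeleton.
# AnimGraph copies the attached parent mesh's pose, then applies an additive head-bone adjustment with a roll offset, weighted by NewVar.
# Each update outside emotes, EventGraph derives NewVar by clamp-mapping the head's yaw delta from reference pose, relative to neck_02, over a negative-angle range.
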AnimBlueprint /BRCosmetics/Characters/Player/Male/Medium/Bodies/M_MED_WingBath_Sturdy/Meshes/Parts/M_MED_WingBath_Sturdy_Head_AnimBP
target_skeleton=&/M_MED_WingBath_Sturdy_FaceAcc_Skeleton
graphs [
 AnimGraph {nodes [
   n1 = "Transform (Modify) Bone - Bone: head"(
    ComponentPose=Local To Component(LocalPose=Copy Pose From Mesh() {bUseAttachedParent=true})
    Alpha=Get NewVar() -> (NewVar:double)
    Translation="0,0,0"
    Rotation="0,0,30"
    Scale="1,1,1"
   ) {
    rotationMode=BMM_Additive
    scaleMode=BMM_Additive
    translationMode=BMM_Additive
    alphaBoolBlend {blendOption=Linear}
    boneToModify {boneName=head}
    rotation {roll=30}
   }
   Output Pose(Result=Component To Local(ComponentPose=@n1))
  ]}
 EventGraph {nodes [
   Event BlueprintUpdateAnimation()
   > Branch(Condition=NOT Boolean(A=Get bIsPlayingEmote()))
   n1 = BreakTransform(InTransform=GetDeltaTransformFromRefPose(self=GetOwnerSkeletalMeshComponent(), BoneName=head, BaseName=neck_02))
   > Set NewVar(NewVar:double=MapRangeClamped(Value=BreakRotator(InRot=@n1.Rotation).Yaw, InRangeA=-20, InRangeB=-45)) -> (Output_Get:double)
   TryGetPawnOwner()
  ]}
]
variables [NewVar {type=number}]
